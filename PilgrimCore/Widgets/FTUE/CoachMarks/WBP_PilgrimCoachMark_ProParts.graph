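# Pilgrim FTUE coach mark telling players that a setting can always display Pro Parts without a connected guitar.
# It extends the Fort sidebar onboarding tooltip, wrapping a UIKit coach mark whose text is set on PreConstruct and which activates on show.
# Both button handlers and OnEndTooltip immediately notify close-animation finished, broadcast OnProPartsTooltipDismissed, then deactivate the coach mark and itself.
# The "Settings" dismiss button is configured hidden, leaving "Got It".
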WidgetBlueprint /PilgrimCore/Widgets/FTUE/CoachMarks/WBP_PilgrimCoachMark_ProParts
graphs [
 EventGraph {nodes [
   "Button Dismiss Clicked (WBP_UIKit_CoachMark)"() {is=title K2Node_ComponentBoundEvent}
   > n1 = HandleDismissClicked(self:Fort Sidebar Onboard Tooltip Widget Object Reference)
   NotifyCloseAnimationFinished(execute=[@n1.then, @n2.then, @n3.then], self:Fort Sidebar Onboard Tooltip Widget Object Reference)
   > Call On Pro Parts Tooltip Dismissed!(execute:Exec, self:WBP Pilgrim Coach Mark Pro Parts Object Reference) -> (then:Exec) {is=title K2Node_CallDelegate}
   > DeactivateWidget(self=Get WBP_UIKit_CoachMark() -> (WBP_UIKit_CoachMark:WBP UIKit Coach Mark Object Reference))
   > DeactivateWidget()
   "Button Other Action Clicked (WBP_UIKit_CoachMark)"() {is=title K2Node_ComponentBoundEvent}
   > n2 = HandleOtherActionClicked(self:Fort Sidebar Onboard Tooltip Widget Object Reference)
   Event OnShowTooltip() {is=title K2Node_Event}
   > ActivateWidget(self=Get WBP_UIKit_CoachMark())
   n3 = Event OnEndTooltip() {is=title K2Node_Event}
   Event PreConstruct()
   n4 = Get WBP_UIKit_CoachMark()
   > SetTextHeader!(execute, self:WBP UIKit Coach Mark Base Object Reference=@n4, Text:Text="LOCTEXT(\"Pro Parts Setting\")") -> (then, Empty:bool)
   > SetTextDescription!(
    execute
    self:WBP UIKit Coach Mark Base Object Reference=@n4
    Text:Text="LOCTEXT(\"There\\'s a setting to always display Pro Parts. Turn this on if you want to see them without a connected guitar.\")"
   ) -> (then, Empty:bool)
  ]}
 OnProPartsTooltipDismissed {nodes [OnProPartsTooltipDismissed() {is=title K2Node_FunctionEntry}]}
]
variables [OnProPartsTooltipDismissed {}]
widgets {
 info {
  inheritedWidgetCount=2
  parentClass=/Script/FortniteUI.FortSidebarOnboardTooltipWidget
  rootWidgetClass=/FortUILibrary/UIKit/CoachMark/WBP_UIKit_CoachMark.WBP_UIKit_CoachMark_C
  widgetCount=3
 }
 widgets [
  WBP_UIKit_CoachMark {
   widgetClassPath=/FortUILibrary/UIKit/CoachMark/WBP_UIKit_CoachMark
   bIsVariable=true
   widget=WBP_UIKit_CoachMark
   properties {
    arrowPosition=Bottom_Left
    bIsFocusable=true
    shouldShowDismissButton=false
    textDismissButton=Settings
    textOtherActionButton=Got It
    widget Width=627
   }
  }
  Dismiss_Button {widgetClassPath=&CommonButtonLegacy, bInherited=true, widget=null}
  OtherAction_Button {widgetClassPath=&CommonButtonLegacy, bInherited=true, widget=null}
 ]
}
